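# Masked, two-sided lit material for leaf effects on mesh particles, Niagara sprites and Nanite meshes.
# A leaf texture is desaturated, brightened and hue-shifted, all driven per particle by a Dynamic Parameter.
# A static switch optionally multiplies in particle color, then an emissive-contribution channel scales the result into the shared FX emissive adjustment function.
# Scalar parameters for specular, roughness and metallic are also defined.
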Material /Game/Effects/Fort_Effects/Materials/Foliage/M_Foliage_Leaf_Mesh_01
graph {
 properties {
  bUsedWithMeshParticles=true
  bUsedWithNanite=true
  bUsedWithNiagaraMeshParticles=true
  bUsedWithNiagaraSprites=true
  blendMode=BLEND_Masked
  pixelDepthOffsetMode=PDOM_Legacy
  twoSided=true
  usedShadingModels=Default Lit
 }
 nodes [
  FeatureLevelSwitch()
  n1 = TextureSampleParameter2D() {parameterName=Texture, texture=/Game/Effects/Fort_Effects/World/Destruction/Textures/T_Prop_Leaf}
  n2 = DynamicParameter() {defaultValue=<rgba 0 1 0 0>, paramNames [Color_Desaturation, Color_ValueMultiplier, Color_EmissiveContribution, DynamicHue]}
  n3 = MaterialFunctionCall(
   Hue Shift Percentage=@n2.DynamicHue
   Texture=Multiply(A=Desaturation(None=@n1.RGB, Fraction=@n2.Color_Desaturation), B=@n2.Color_ValueMultiplier)
  ) {materialFunction=&HueShift}
  n4 = Multiply(A=@n3.Result, B=ParticleColor().RGB)
  MaterialFunctionCall(Emissive=Multiply(A=StaticSwitchParameter(True=@n4, False=@n3.Result) {parameterName=UseParticleColor}, B=@n2.Color_EmissiveContribution)) {materialFunction=&MF_FX_EmissiveAdjustments}
  ScalarParameter() {parameterName=Hue, sliderMax=1}
  ScalarParameter() {defaultValue=0.5, parameterName=Specular}
  ScalarParameter() {defaultValue=0.55, parameterName=Roughness}
  ScalarParameter() {parameterName=Metallic}
 ]
}
